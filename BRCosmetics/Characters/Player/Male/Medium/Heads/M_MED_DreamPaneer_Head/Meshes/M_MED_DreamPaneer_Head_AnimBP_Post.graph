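# Post-process AnimBlueprint for the DreamPaneer medium male player head that nudges the teeth and tongue bones on the shared Fortnite head skeleton.
# The AnimGraph chains additive bone-space translations on the incoming pose: lower teeth, upper teeth, lower teeth again, then tongue, at constant offsets and full alpha.
# The update event stores the jaw_forward_pose curve, power-shaped and clamped to 0–1, in jawForward, which the AnimGraph does not read.
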AnimBlueprint /BRCosmetics/Characters/Player/Male/Medium/Heads/M_MED_DreamPaneer_Head/Meshes/M_MED_DreamPaneer_Head_AnimBP_Post
target_skeleton=/Game/Characters/Player/Common/Fortnite_Base_Head/Fortnite_Base_Head_Export_Skeleton
graphs [
 AnimGraph {nodes [
   n1 = "Transform (Modify) Bone - Bone: teeth_lower"!(ComponentPose=Local To Component(LocalPose=Input Pose()), Alpha=1, Translation="0,0,-0.2") -> (Pose) {translationMode=BMM_Additive, translationSpace=BCS_BoneSpace, boneToModify {boneName=teeth_lower}, translation {z=-0.2}}
   n2 = "Transform (Modify) Bone - Bone: teeth_upper"!(ComponentPose=@n1, Alpha=1, Translation="0,-0.12,-0.1") -> (Pose) {translationMode=BMM_Additive, translationSpace=BCS_BoneSpace, boneToModify {boneName=teeth_upper}, translation {y=-0.12, z=-0.1}}
   n3 = "Transform (Modify) Bone - Bone: teeth_lower"(ComponentPose=@n2, Alpha=1, Translation="0,-0.12,0") {translationMode=BMM_Additive, translationSpace=BCS_BoneSpace, boneToModify {boneName=teeth_lower}, translation {y=-0.12}}
   n4 = "Transform (Modify) Bone - Bone: tongue"!(ComponentPose=@n3, Alpha=1, Translation="0,-0.1141,0.0372") -> (Pose) {translationMode=BMM_Additive, translationSpace=BCS_BoneSpace, boneToModify {boneName=tongue}, translation {y=-0.1141, z=0.0372}}
   Output Pose(Result=Component To Local(ComponentPose=@n4))
  ]}
 EventGraph {nodes [
   Event BlueprintUpdateAnimation()
   n1 = GetCurveValue(CurveName=jaw_forward_pose)
   > Set jawForward(jawForward:double="Clamp (Float)"(Value=Power!(Base:double=@n1, Exp:double=0.7) -> (ReturnValue:double) {is=title K2Node_CallFunction}, Min=0, Max=1)) -> (Output_Get:double)
  ]}
]
variables [jawForward {type=number}]
